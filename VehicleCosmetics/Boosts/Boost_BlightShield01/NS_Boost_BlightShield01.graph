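# Niagara system for the BlightShield01 vehicle boost cosmetic, drawing wavy ribbon trails and short-lived mesh/sprite effects at the boost sockets.
# Two ribbon emitters compute four phase-offset sine/cosine positions from user wave parameters, mirroring X/Y by spawn group, with curves shaping amplitude, width and material parameters.
# Two mesh/sprite emitters partition particles across filtered skeletal sockets, offset by the car offset.
# Painted-variant override modules are disabled in all emitters.
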NiagaraSystem /VehicleCosmetics/Boosts/Boost_BlightShield01/NS_Boost_BlightShield01
dependencies {
 usedDataInterfaces [&NiagaraDataInterfaceCurve, &NiagaraDataInterfaceArrayFloat, &NiagaraDataInterfaceSkeletalMesh]
 usedDynamicInputs []
 usedModules [
  &self:SystemSpawnScript.SystemScriptSource.SystemScriptGraph.NiagaraNodeAssignment_0.SetVariables_#1
  &SystemState
 ]
}
emitters {
 NE_BlightShield_OPT002 {
  emitterUpdateScript [
   &EmitterState(Life Cycle Mode=::NewEnumerator1, Loop Duration=1, UseLoopDelay=false)
   &VC_FilteredSocketCount()
   &VC_RateQualityHandler(High Spawn Rate=60, Low Spawn Rate=30, Use Custom Rate?=true)
   &VC_SpawnRibbon_v1(Number of Ribbons=$Emitter.PortCount, SpawnRate=$Emitter.Spawn Rate)
  ]
  particleSpawnScript [
   &VC_PaintedVariantsOverride() {enabled=false}
   &VC_InitializeRibbon_v1(
    Custom Offset=true
    Enable Skeletal Position=true
    Lifetime=1.25
    Local Vector Offset=&AddVectorToPosition(Position=<-76 0 0>, Vector=Linked() {linkedVariable {name=User.DTCarOffset, type=NiagaraPosition}})
    Offset Amount=1
    Ribbon Width=30
   )
   &AddVelocity(Velocity=<-1000 0 0>) {module=AddVelocity001}
  ]
  particleUpdateScript [
   &ParticleState()
   &InheritSourceMovement(
    Applied Position Scale=&VectorFromFloat(Value=&FloatFromCurve(FloatCurve=DataInterface() {propertyValues {Curve {keys [
          {interpMode=RCIM_Cubic, tangentMode=RCTM_Break, time=0, value=1}
          {arriveTangent=-2.9025, interpMode=RCIM_Cubic, tangentMode=RCTM_Break, time=1, value=0}
         ]}}}))
    ApplyForce=false
    ApplyPosition=true
    ApplyPositionFalloff=false
   )
   /SetVariables(
    Particles.Ribbon_BlackColor:LinearColor=<rgba 0 0 0 1>
    Particles.WaveScaleCurve:NiagaraFloat=&FloatFromCurve(FloatCurve=DataInterface() {propertyValues {Curve {keys [
         {arriveTangent=67.3191, interpMode=RCIM_Cubic, leaveTangent=67.3191, tangentMode=RCTM_User, time=0.0092, value=0}
         {
          arriveTangent=0.7219
          arriveTangentWeight=0.0795
          interpMode=RCIM_Cubic
          leaveTangent=0.1806
          leaveTangentWeight=0.0696
          tangentMode=RCTM_Break
          tangentWeightMode=RCTWM_WeightedBoth
          time=0.125
          value=3.25
         }
         {arriveTangent=0.1988, interpMode=RCIM_Cubic, leaveTangent=0.1988, tangentMode=RCTM_User, time=0.45, value=1.7}
         {
          arriveTangent=-0.5881
          arriveTangentWeight=0.5973
          interpMode=RCIM_Cubic
          leaveTangent=-0.5881
          leaveTangentWeight=0.5973
          tangentMode=RCTM_User
          time=0.75
          value=0.2796
         }
        ]}}})
   ) {isSetParameters=true, module=SetVariables_#2}
   /SetVariables(
    Particles.CustomPosition_R1:NiagaraPosition=&AddVectorToPosition(
     Position=$Particles.Position
     Vector=&Multiply_VectorByFloat(
      Vector=&MakeVector(
       X=&Sine(
        Normalized Angle=&Add_Float(A=0, B=&Subtract_Float(A=$Emitter.NormalizedLoopAge, B=$Particles.NormalizedAge))
        Period=$User.WavePeriod
        Scale=&Multiply_Float(
         A=&SelectFloatFromArray(Direct Array Index=$Particles.SpawnGroup, Float Selection Array=DataInterface() {propertyValues {FloatData [1,-1]}})
         B=$User.WaveMaxScale
        )
        Bias=0
       )
       Y=&Sine(
        Normalized Angle=&Add_Float(A=0, B=&Subtract_Float(A=$Emitter.NormalizedLoopAge, B=$Particles.NormalizedAge))
        Period=$User.WavePeriod
        Scale=&Multiply_Float(
         A=&SelectFloatFromArray(Direct Array Index=$Particles.SpawnGroup, Float Selection Array=DataInterface() {propertyValues {FloatData [1,-1]}})
         B=$User.WaveMaxScale
        )
        Bias=0
       )
       Z=&Cosine(
        Normalized Angle=&Add_Float(A=0, B=&Subtract_Float(A=$Emitter.NormalizedLoopAge, B=$Particles.NormalizedAge))
        Period=$User.WavePeriod
        Scale=$User.WaveMaxScale
        Bias=0
       )
      )
      Float=$Particles.WaveScaleCurve
     )
    )
    Particles.CustomPosition_R2:NiagaraPosition=&AddVectorToPosition(
     Position=$Particles.Position
     Vector=&Multiply_VectorByFloat(
      Vector=&MakeVector(
       X=&Sine(
        Normalized Angle=&Add_Float(
         A=&Multiply_Float(A=$User.WavePeriod, B=0.5)
         B=&Subtract_Float(A=$Emitter.NormalizedLoopAge, B=$Particles.NormalizedAge)
        )
        Period=$User.WavePeriod
        Scale=&Multiply_Float(
         A=&SelectFloatFromArray(Direct Array Index=$Particles.SpawnGroup, Float Selection Array=DataInterface() {propertyValues {FloatData [1,-1]}})
         B=$User.WaveMaxScale
        )
        Bias=0
       )
       Y=&Sine(
        Normalized Angle=&Add_Float(
         A=&Multiply_Float(A=$User.WavePeriod, B=0.5)
         B=&Subtract_Float(A=$Emitter.NormalizedLoopAge, B=$Particles.NormalizedAge)
        )
        Period=$User.WavePeriod
        Scale=&Multiply_Float(
         A=&SelectFloatFromArray(Direct Array Index=$Particles.SpawnGroup, Float Selection Array=DataInterface() {propertyValues {FloatData [1,-1]}})
         B=$User.WaveMaxScale
        )
        Bias=0
       )
       Z=&Cosine(
        Normalized Angle=&Add_Float(
         A=&Multiply_Float(A=$User.WavePeriod, B=0.5)
         B=&Subtract_Float(A=$Emitter.NormalizedLoopAge, B=$Particles.NormalizedAge)
        )
        Period=$User.WavePeriod
        Scale=$User.WaveMaxScale
        Bias=0
       )
      )
      Float=$Particles.WaveScaleCurve
     )
    )
   ) {isSetParameters=true, module=SetVariables_#3}
   /SetVariables(Particles.DynamicMP_R2:Vector4f=&MakeVector4(
     X=&FloatFromCurve(FloatCurve=DataInterface() {propertyValues {Curve {
         defaultValue=1
         keys [
          {tangentMode=RCTM_User, time=0, value=1}
          {arriveTangent=-0.0633, interpMode=RCIM_Cubic, leaveTangent=-0.0633, tangentMode=RCTM_User, time=0.65, value=1}
          {arriveTangent=-0.0699, interpMode=RCIM_Cubic, leaveTangent=-0.0699, tangentMode=RCTM_User, time=0.75, value=-2}
         ]
        }}})
     Y=0
     Z=0
     W=0
    )) {isSetParameters=true, module=SetVariables_#4}
   &DynamicMaterialParameters(
    Index 0 Param 1=&FloatFromCurve(FloatCurve=DataInterface() {propertyValues {Curve {
        defaultValue=1
        keys [
         {arriveTangent=-0.0405, interpMode=RCIM_Cubic, leaveTangent=-0.0405, tangentMode=RCTM_User, time=0.45, value=1}
         {arriveTangent=-0.0201, interpMode=RCIM_Cubic, leaveTangent=-0.0201, tangentMode=RCTM_User, time=1, value=-2}
        ]
       }}})
    Module.Index0Float0Write=true
    Module.Index0Float1Write=true
    Module.Index0Float2Write=true
    Module.Index0Float3Write=true
   )
   &SolveForcesAndVelocity(Clamp Velocity=false, Limit Acceleration=false)
   &ScaleRibbonWidth(Ribbon Width Scale=&FloatFromCurve(FloatCurve=DataInterface() {propertyValues {Curve {
        defaultValue=1
        keys [
         {arriveTangent=16.0927, interpMode=RCIM_Cubic, leaveTangent=16.0927, tangentMode=RCTM_User, time=0.02, value=0}
         {
          arriveTangent=0.1655
          arriveTangentWeight=0.0629
          interpMode=RCIM_Cubic
          leaveTangent=0.1655
          leaveTangentWeight=0.0629
          tangentMode=RCTM_User
          tangentWeightMode=RCTWM_WeightedBoth
          time=0.1251
          value=0.9549
         }
         {arriveTangent=-1.191, interpMode=RCIM_Cubic, leaveTangent=-1.191, tangentMode=RCTM_User, time=0.3862, value=0.9365}
         {arriveTangent=-0.1455, interpMode=RCIM_Cubic, leaveTangent=-0.1455, tangentMode=RCTM_User, time=0.6182, value=0.6212}
         {arriveTangent=0.4201, interpMode=RCIM_Cubic, leaveTangent=0.4201, tangentMode=RCTM_User, time=0.75, value=0}
        ]
       }}}))
   &VC_UpdateRibbon_v1(Enable Skeletal Position=true)
  ]
  rendererClasses [&NiagaraRibbonRendererProperties]
  renderers [
   {rendererClass=&NiagaraRibbonRendererProperties}
   {rendererClass=&NiagaraRibbonRendererProperties, rendererIndex=1}
   {rendererClass=&NiagaraRibbonRendererProperties, rendererIndex=2}
   {rendererClass=&NiagaraRibbonRendererProperties, rendererIndex=3}
  ]
 }
 NE_BlightShield_OPT003 {
  emitterUpdateScript [
   &EmitterState(Life Cycle Mode=::NewEnumerator1, Loop Duration=1, UseLoopDelay=false)
   &VC_FilteredSocketCount()
   &VC_RateQualityHandler(High Spawn Rate=60, Low Spawn Rate=30, Use Custom Rate?=true)
   &VC_SpawnRibbon_v1(Number of Ribbons=$Emitter.PortCount, SpawnRate=$Emitter.Spawn Rate)
  ]
  particleSpawnScript [
   &VC_PaintedVariantsOverride() {enabled=false}
   &VC_InitializeRibbon_v1(
    Custom Offset=true
    Enable Skeletal Position=true
    Lifetime=1.25
    Local Vector Offset=&AddVectorToPosition(Position=<-76 0 0>, Vector=Linked() {linkedVariable {name=User.DTCarOffset, type=NiagaraPosition}})
    Offset Amount=1
    Ribbon Width=30
   )
   &AddVelocity(Velocity=<-1000 0 0>) {module=AddVelocity001}
  ]
  particleUpdateScript [
   &ParticleState()
   &InheritSourceMovement(
    Applied Position Scale=&VectorFromFloat(Value=&FloatFromCurve(FloatCurve=DataInterface() {propertyValues {Curve {keys [
          {interpMode=RCIM_Cubic, tangentMode=RCTM_Break, time=0, value=1}
          {arriveTangent=-2.9025, interpMode=RCIM_Cubic, tangentMode=RCTM_Break, time=1, value=0}
         ]}}}))
    ApplyForce=false
    ApplyPosition=true
    ApplyPositionFalloff=false
   )
   /SetVariables(
    Particles.Ribbon_BlackColor:LinearColor=<rgba 0 0 0 1>
    Particles.WaveScaleCurve:NiagaraFloat=&FloatFromCurve(FloatCurve=DataInterface() {propertyValues {Curve {keys [
         {arriveTangent=67.3191, interpMode=RCIM_Cubic, leaveTangent=67.3191, tangentMode=RCTM_User, time=0.0092, value=0}
         {
          arriveTangent=0.7219
          arriveTangentWeight=0.0795
          interpMode=RCIM_Cubic
          leaveTangent=0.1806
          leaveTangentWeight=0.0696
          tangentMode=RCTM_Break
          tangentWeightMode=RCTWM_WeightedBoth
          time=0.125
          value=3.25
         }
         {arriveTangent=0.1988, interpMode=RCIM_Cubic, leaveTangent=0.1988, tangentMode=RCTM_User, time=0.45, value=1.7}
         {
          arriveTangent=-0.5881
          arriveTangentWeight=0.5973
          interpMode=RCIM_Cubic
          leaveTangent=-0.5881
          leaveTangentWeight=0.5973
          tangentMode=RCTM_User
          time=0.75
          value=0.2796
         }
        ]}}})
   ) {isSetParameters=true, module=SetVariables_#2}
   /SetVariables(
    Particles.CustomPosition_R3:NiagaraPosition=&AddVectorToPosition(
     Position=$Particles.Position
     Vector=&Multiply_VectorByFloat(
      Vector=&MakeVector(
       X=&Sine(
        Normalized Angle=&Add_Float(
         A=&Multiply_Float(A=$User.WavePeriod, B=0.25)
         B=&Subtract_Float(A=$Emitter.NormalizedLoopAge, B=$Particles.NormalizedAge)
        )
        Period=$User.WavePeriod
        Scale=&Multiply_Float(
         A=&SelectFloatFromArray(Direct Array Index=$Particles.SpawnGroup, Float Selection Array=DataInterface() {propertyValues {FloatData [1,-1]}})
         B=$User.WaveMaxScale
        )
        Bias=0
       )
       Y=&Sine(
        Normalized Angle=&Add_Float(
         A=&Multiply_Float(A=$User.WavePeriod, B=0.25)
         B=&Subtract_Float(A=$Emitter.NormalizedLoopAge, B=$Particles.NormalizedAge)
        )
        Period=$User.WavePeriod
        Scale=&Multiply_Float(
         A=&SelectFloatFromArray(Direct Array Index=$Particles.SpawnGroup, Float Selection Array=DataInterface() {propertyValues {FloatData [1,-1]}})
         B=$User.WaveMaxScale
        )
        Bias=0
       )
       Z=&Cosine(
        Normalized Angle=&Add_Float(
         A=&Multiply_Float(A=$User.WavePeriod, B=0.25)
         B=&Subtract_Float(A=$Emitter.NormalizedLoopAge, B=$Particles.NormalizedAge)
        )
        Period=$User.WavePeriod
        Scale=$User.WaveMaxScale
        Bias=0
       )
      )
      Float=$Particles.WaveScaleCurve
     )
    )
    Particles.CustomPosition_R4:NiagaraPosition=&AddVectorToPosition(
     Position=$Particles.Position
     Vector=&Multiply_VectorByFloat(
      Vector=&MakeVector(
       X=&Sine(
        Normalized Angle=&Add_Float(
         A=&Multiply_Float(A=$User.WavePeriod, B=0.75)
         B=&Subtract_Float(A=$Emitter.NormalizedLoopAge, B=$Particles.NormalizedAge)
        )
        Period=$User.WavePeriod
        Scale=&Multiply_Float(
         A=&SelectFloatFromArray(Direct Array Index=$Particles.SpawnGroup, Float Selection Array=DataInterface() {propertyValues {FloatData [1,-1]}})
         B=$User.WaveMaxScale
        )
        Bias=0
       )
       Y=&Sine(
        Normalized Angle=&Add_Float(
         A=&Multiply_Float(A=$User.WavePeriod, B=0.75)
         B=&Subtract_Float(A=$Emitter.NormalizedLoopAge, B=$Particles.NormalizedAge)
        )
        Period=$User.WavePeriod
        Scale=&Multiply_Float(
         A=&SelectFloatFromArray(Direct Array Index=$Particles.SpawnGroup, Float Selection Array=DataInterface() {propertyValues {FloatData [1,-1]}})
         B=$User.WaveMaxScale
        )
        Bias=0
       )
       Z=&Cosine(
        Normalized Angle=&Add_Float(
         A=&Multiply_Float(A=$User.WavePeriod, B=0.75)
         B=&Subtract_Float(A=$Emitter.NormalizedLoopAge, B=$Particles.NormalizedAge)
        )
        Period=$User.WavePeriod
        Scale=$User.WaveMaxScale
        Bias=0
       )
      )
      Float=$Particles.WaveScaleCurve
     )
    )
   ) {isSetParameters=true, module=SetVariables_#3}
   /SetVariables(Particles.DynamicMP_R3:Vector4f=&MakeVector4(
     X=&FloatFromCurve(FloatCurve=DataInterface() {propertyValues {Curve {
         defaultValue=1
         keys [
          {arriveTangent=-0.165, interpMode=RCIM_Cubic, leaveTangent=-0.165, tangentMode=RCTM_User, time=0.225, value=2}
          {arriveTangent=-0.1372, interpMode=RCIM_Cubic, leaveTangent=-0.1372, tangentMode=RCTM_User, time=0.425, value=-2}
         ]
        }}})
     Y=0
     Z=0
     W=0
    )) {isSetParameters=true, module=SetVariables_#4}
   &DynamicMaterialParameters(
    Index 0 Param 1=&FloatFromCurve(FloatCurve=DataInterface() {propertyValues {Curve {
        defaultValue=1
        keys [
         {arriveTangent=-0.0405, interpMode=RCIM_Cubic, leaveTangent=-0.0405, tangentMode=RCTM_User, time=0.45, value=1}
         {arriveTangent=-0.0201, interpMode=RCIM_Cubic, leaveTangent=-0.0201, tangentMode=RCTM_User, time=1, value=-2}
        ]
       }}})
    Module.Index0Float0Write=true
    Module.Index0Float1Write=true
    Module.Index0Float2Write=true
    Module.Index0Float3Write=true
   )
   &SolveForcesAndVelocity(Clamp Velocity=false, Limit Acceleration=false)
   &ScaleRibbonWidth(Ribbon Width Scale=&FloatFromCurve(FloatCurve=DataInterface() {propertyValues {Curve {
        defaultValue=1
        keys [
         {arriveTangent=16.0927, interpMode=RCIM_Cubic, leaveTangent=16.0927, tangentMode=RCTM_User, time=0.02, value=0}
         {
          arriveTangent=0.1655
          arriveTangentWeight=0.0629
          interpMode=RCIM_Cubic
          leaveTangent=0.1655
          leaveTangentWeight=0.0629
          tangentMode=RCTM_User
          tangentWeightMode=RCTWM_WeightedBoth
          time=0.1251
          value=0.9549
         }
         {arriveTangent=-1.061, interpMode=RCIM_Cubic, leaveTangent=-1.061, tangentMode=RCTM_User, time=0.2496, value=0.4976}
         {arriveTangent=-0.0886, interpMode=RCIM_Cubic, leaveTangent=-0.0886, tangentMode=RCTM_User, time=0.3559, value=0.1114}
         {arriveTangent=-0.1455, interpMode=RCIM_Cubic, leaveTangent=-0.1455, tangentMode=RCTM_User, time=0.7241, value=0.793}
         {arriveTangent=0.4201, interpMode=RCIM_Cubic, leaveTangent=0.4201, tangentMode=RCTM_User, time=0.9, value=0}
        ]
       }}}))
   &VC_UpdateRibbon_v1(Enable Skeletal Position=true)
  ]
  rendererClasses [&NiagaraRibbonRendererProperties]
  renderers [
   {rendererClass=&NiagaraRibbonRendererProperties}
   {rendererClass=&NiagaraRibbonRendererProperties, rendererIndex=1}
   {rendererClass=&NiagaraRibbonRendererProperties, rendererIndex=2}
   {rendererClass=&NiagaraRibbonRendererProperties, rendererIndex=3}
  ]
 }
 NE_VC_Boost_Sprite_Parent {
  emitterUpdateScript [
   &EmitterState(Inactive Response=::NewEnumerator1, Life Cycle Mode=::NewEnumerator1, Loop Duration=1, UseLoopDelay=false)
   &VC_FilteredSocketCount()
   &SpawnBurst_Instantaneous(Spawn Count=4, Use Spawn Probability=false, UseLoopCountLimit=false) {enabled=false}
   &SpawnRate(SpawnRate=8, Use Spawn Probability=false)
  ]
  particleSpawnScript [
   &VC_PaintedVariantsOverride() {enabled=false}
   &InitializeParticle(
    Color=<rgba 20 20 20 1>
    Lifetime=0.25
    Mesh Scale Mode=::NewEnumerator3
    Mesh Uniform Scale=0.775
    Sprite Rotation Mode=::NewEnumerator2
    Sprite Size Mode=::NewEnumerator3
    UsePositionOffset=false
    Write Material Random=true
   )
   &PartitionParticles(Number Of Partitions=$Emitter.PortCount)
   &InitialMeshOrientation(
    Mesh Orientation Mode=::NewEnumerator2
    Rotation=&RandomRangeVector(Minimum=<-1 0 0>, Maximum=<1 0 0>)
    Use Rotation Vector=true
   )
   &SkeletalMeshLocation(
    Bone / Socket Index=$Particles.PartitionParticles.Partition
    Bone / Socket Sampling Mode=::NewEnumerator5
    MaskSpawnGroup=false
    Mesh Sampling Type=::NewEnumerator6
    Sampled Position Offset=&AddVectorToPosition(Position=$User.DTCarOffset, Vector=<-115 0 0>)
    Skeletal Mesh=$System.SkeletalMesh
   )
   &AddVelocity(Velocity=<-1500 0 0>) {enabled=false}
  ]
  particleUpdateScript [
   &ParticleState()
   &SolveForcesAndVelocity(Clamp Velocity=false, Limit Acceleration=false)
  ]
  rendererClasses [&NiagaraMeshRendererProperties, &NiagaraSpriteRendererProperties]
  renderers [
   {rendererClass=&NiagaraMeshRendererProperties, rendererIndex=0}
   {rendererClass=&NiagaraSpriteRendererProperties, rendererIndex=1}
  ]
 }
 NE_VC_Boost_Sprite_Parent002 {
  emitterUpdateScript [
   &EmitterState(Inactive Response=::NewEnumerator1, Life Cycle Mode=::NewEnumerator1, Loop Duration=1, UseLoopDelay=false)
   &VC_FilteredSocketCount()
   &SpawnBurst_Instantaneous(Spawn Count=4, Use Spawn Probability=false, UseLoopCountLimit=false) {enabled=false}
   &SpawnRate(SpawnRate=60, Use Spawn Probability=false)
  ]
  particleSpawnScript [
   &VC_PaintedVariantsOverride() {enabled=false}
   &InitializeParticle(
    Color=<rgba 2.5 2.5 2.5 1>
    Lifetime=0.025
    Mesh Scale Max=<0.35 0.5 0.35>
    Mesh Scale Min=<0.3 0.325 0.3>
    Mesh Scale Mode=::NewEnumerator2
    Sprite Rotation Mode=::NewEnumerator2
    Sprite Size Mode=::NewEnumerator3
    UsePositionOffset=false
    Write Material Random=true
   )
   &PartitionParticles(Number Of Partitions=$Emitter.PortCount)
   &InitialMeshOrientation(
    Mesh Orientation Mode=::NewEnumerator2
    Rotation=&RandomRangeVector(Minimum=<-1 0 0>, Maximum=<1 0 0>)
    Use Rotation Vector=true
   )
   &SkeletalMeshLocation(
    Bone / Socket Index=$Particles.PartitionParticles.Partition
    Bone / Socket Sampling Mode=::NewEnumerator5
    MaskSpawnGroup=false
    Mesh Sampling Type=::NewEnumerator6
    Sampled Position Offset=&AddVectorToPosition(Position=$User.DTCarOffset, Vector=<-55 0 0>)
    Skeletal Mesh=$System.SkeletalMesh
   )
   &AddVelocity(Velocity=<-1500 0 0>) {enabled=false}
  ]
  particleUpdateScript [
   &ParticleState()
   &SolveForcesAndVelocity(Clamp Velocity=false, Limit Acceleration=false)
  ]
  rendererClasses [&NiagaraMeshRendererProperties, &NiagaraSpriteRendererProperties]
  renderers [
   {rendererClass=&NiagaraMeshRendererProperties, rendererIndex=0}
   {rendererClass=&NiagaraSpriteRendererProperties, rendererIndex=1}
  ]
 }
}
summary {
 systemName=NS_Boost_BlightShield01
 emitters [
  {
   emitterName=NE_VC_Boost_Sprite_Parent
   rendererClasses [&NiagaraMeshRendererProperties, &NiagaraSpriteRendererProperties]
  }
  {emitterName=NE_BlightShield_OPT002, rendererClasses [&NiagaraRibbonRendererProperties]}
  {emitterName=NE_BlightShield_OPT003, rendererClasses [&NiagaraRibbonRendererProperties]}
  {
   emitterName=NE_VC_Boost_Sprite_Parent002
   rendererClasses [&NiagaraMeshRendererProperties, &NiagaraSpriteRendererProperties]
  }
 ]
 userVariables [
  User.bIsFrontendPreview {type=NiagaraBool}
  User.bIsPainted {type=NiagaraBool}
  User.ColorVariantID {type=NiagaraInt32}
  User.DTCarOffset {type=NiagaraPosition, description="", defaultValue {}}
  User.PaintedColor {type=LinearColor, description="", defaultValue {struct=LinearColor, value=<rgba 1 1 1 1>}}
  User.PaintedScalar {type=NiagaraFloat, defaultValue=1, description=""}
  User.PortCount {type=NiagaraInt32, defaultValue=2}
  User.WaveMaxScale {type=NiagaraFloat, defaultValue=10, description=""}
  User.WavePeriod {type=NiagaraFloat, defaultValue=1, description=""}
 ]
}
